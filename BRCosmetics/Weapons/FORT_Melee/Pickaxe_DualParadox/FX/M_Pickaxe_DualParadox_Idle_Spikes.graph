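# Translucent, unlit, two-sided Niagara mesh-particle material for the idle aura spikes on the Dual Paradox pickaxe.
# A spike texture, sampled with the panned U coordinate on both axes, is masked by UV edge gradients, a switchable spike-direction gradient and view-angle fade.
# Depth fade and particle alpha shape the opacity, while a curve-atlas color ramp along U, tinted by particle color, drives emissive.
# Several other fresnel and gradient chains are left unconnected.
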Material /BRCosmetics/Weapons/FORT_Melee/Pickaxe_DualParadox/FX/M_Pickaxe_DualParadox_Idle_Spikes
graph {
 properties {
  bUsedWithNiagaraMeshParticles=true
  blendMode=BLEND_Translucent
  pixelDepthOffsetMode=PDOM_Legacy
  shadingModel=MSM_Unlit
  twoSided=true
  usedShadingModels=Unlit
 }
 nodes [
  Constant() {r=1}
  n1 = ScalarParameter() {defaultValue=1, parameterName="FadeSensitivity (optional) (S)"}
  n2 = DotProduct(A=Transform(None=Constant3Vector() {constant=<rgba 0 -1 0 0>}), B=CameraVectorWS())
  n3 = Saturate(None=Add(A=Abs(None=@n2), B=ScalarParameter() {parameterName=FadegradientOffset}))
  n4 = ScalarParameter() {defaultValue=0.5, parameterName=Fade gradient Contrast}
  LinearInterpolate(
   A=@n1
   B=LinearInterpolate(
    A=@n1
    B=ScalarParameter() {defaultValue=1, parameterName=FadeSensitivityB}
    Alpha=MaterialFunctionCall(In=@n3, Contrast=@n4) {materialFunction=&CheapContrast}.Result
   )
   Alpha=ScalarParameter() {parameterName=Use custom angle fade}
  )
  n5 = ComponentMask(None=TextureCoordinate()) {g=true}
  Multiply(A=Power(Base=OneMinus(None=@n5)) {constExponent=3}) {constB=10}
  n6 = VectorParameter() {group=Mask, parameterName=Mask_Speed, defaultValue=<rgba 0 2.2 0 0>}
  n7 = ComponentMask(None=VectorParameter() {parameterName=UV_Scale, defaultValue=<rgba 1 3 0 0>}.RGB) {g=true, r=true}
  n8 = TextureCoordinate()
  n9 = MaterialFunctionCall(Float2=Panner(Coordinate=Multiply(A=@n7, B=@n8), Speed=ComponentMask(None=@n6.RGB) {g=true, r=true})) {materialFunction=&BreakOutFloat2Components}
  n10 = OneMinus(None=MaterialFunctionCall(Float2=TextureCoordinate()) {materialFunction=&BreakOutFloat2Components}.G)
  n11 = ScalarParameter() {defaultValue=-0.3, group=Mask, parameterName=Edge_Mask_UV_scale}
  n12 = Add(A=Multiply(A=Subtract(A=@n10) {constB=0.65}) {constB=2}, B=@n11) {constB=-0.5}
  n13 = Add(B=@n11) {constA=1}
  n14 = Power(Base=Multiply(A=Saturate(None=@n12), B=Divide(B=@n13) {constB=1}) {constB=2}) {constExponent=4}
  Multiply(A=@n9.R, B=Add(A=Multiply(A=@n14) {constB=5}))
  n15 = ParticleColor()
  n16 = TextureSampleParameter2D(UVs=MaterialFunctionCall(X=@n9.R, Y=@n9.R) {materialFunction=&MakeFloat2}.Result) {parameterName=Tex, texture=&/T_DualParadox_pickaxe_aura_spikes}
  n17 = TextureCoordinate()
  n18 = Multiply(
   A=Power(Base=ComponentMask(None=@n17) {r=true}, Exp=ScalarParameter() {defaultValue=1.1, parameterName=Left mask power}) {constExponent=1}
   B=ScalarParameter() {defaultValue=2, parameterName=Left mask Edge}
  ) {constB=2}
  n19 = Power(Base=ComponentMask(None=TextureCoordinate()) {g=true}, Exp=ScalarParameter() {defaultValue=1.5, parameterName=Top mask power}) {constExponent=1.5}
  n20 = ScalarParameter() {defaultValue=10, parameterName=Mask edge}
  n21 = Multiply(
   A=Multiply(A=StaticComponentMaskParameter(None=@n16.RGB) {defaultR=true, parameterName=Mask}, B=Saturate(None=@n18))
   B=Saturate(None=Multiply(A=@n19, B=@n20) {constB=10})
  )
  n22 = OneMinus(None=MaterialFunctionCall() {materialFunction=&FadeBasedOnViewAngle}.Result)
  n23 = ScalarParameter() {defaultValue=2, parameterName=FadeSensitivity power}
  n24 = Saturate(None=OneMinus(None=Power(Base=@n22, Exp=@n23)))
  n25 = TextureCoordinate()
  n26 = Power(Base=OneMinus(None=ComponentMask(None=@n25) {g=true})) {constExponent=4}
  n27 = ComponentMask(None=TextureCoordinate()) {g=true}
  n28 = Multiply(A=Power(Base=OneMinus(None=@n27)) {constExponent=4}) {constB=100}
  n29 = LinearInterpolate(
   A=Saturate(None=Multiply(A=@n26) {constB=70})
   B=OneMinus(None=Saturate(None=@n28))
   Alpha=ScalarParameter() {parameterName=Spikes Direction}
  )
  n30 = ScalarParameter() {defaultValue=1, parameterName=Use Blade top mask}
  n31 = DepthFade(
   Opacity=Multiply(A=Multiply(A=@n21, B=@n24), B=LinearInterpolate(B=@n29, Alpha=@n30) {constA=1})
   FadeDistance=ScalarParameter() {defaultValue=3, parameterName=FadeDistance}
  ) {fadeDistanceDefault=3}
  Reroute(None=Multiply(A=Multiply(A=@n15.A, B=@n31), B=ScalarParameter() {defaultValue=1, parameterName=Main Alpha}))
  n32 = CurveAtlasRowParameter(CurveTime=ComponentMask(None=TextureCoordinate()) {r=true}) {
   atlas=/Game/Effects/Fort_Effects/Emotes/DualParadox/Curve_Atlas_DBZ_Emote
   curve=/Game/Weapons/FORT_Melee/Pickaxe_DualParadox/FX/CV_Pickaxe_DualParadox_color_Curve
   defaultValue=2
  }
  n33 = Multiply(A=@n32)
  Reroute(None=MaterialFunctionCall(Emissive=Multiply(A=@n33, B=@n15.RGB)) {materialFunction=&MF_FX_EmissiveAdjustments}.Result)
  n34 = Constant3Vector() {constant=<rgba 0 0 1 0>}
  n35 = CameraVectorWS()
  n36 = OneMinus(None=Power(Base=DotProduct(A=@n34, B=@n35)) {constExponent=7})
  n37 = Constant() {r=1}
  Saturate(None=OneMinus(None=Power(Base=@n36, Exp=@n37)))
  n38 = CameraVectorWS()
  n39 = VertexNormalWS()
  n40 = Power(Base=Abs(None=DotProduct(A=@n38, B=@n39)), Exp=@n1)
  n41 = Constant3Vector() {constant=<rgba 0 0 1 0>}
  n42 = CameraVectorWS()
  n43 = ScalarParameter() {parameterName=Param_1}
  n44 = OneMinus(None=Power(Base=DotProduct(A=@n41, B=@n42), Exp=@n43) {constExponent=7})
  n45 = ScalarParameter() {parameterName=Param_2}
  n46 = Saturate(None=OneMinus(None=Power(Base=@n44, Exp=@n45)))
  n47 = ScalarParameter() {parameterName=Param_3}
  Saturate(None=Add(A=Clamp(None=@n40), B=LinearInterpolate(B=@n46, Alpha=@n47)) {constB=0})
  n48 = Constant3Vector() {constant=<rgba 0 0 -1 0>}
  n49 = CameraVectorWS()
  n50 = OneMinus(None=Power(Base=DotProduct(A=@n48, B=@n49), Exp=@n43) {constExponent=7})
  Saturate(None=OneMinus(None=Power(Base=@n50, Exp=@n45)))
 ]
}
