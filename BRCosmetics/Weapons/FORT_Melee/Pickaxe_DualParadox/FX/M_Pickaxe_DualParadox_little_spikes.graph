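# Translucent unlit material for Niagara mesh particles, drawing the little spikes effect of the DualParadox pickaxe.
# Emissive color comes from the shared MFC_DualParadox_PurpleNoise function, tinted by particle color and passed through MF_FX_EmissiveAdjustments.
# An opacity chain multiplies particle alpha by a steep UV-V gradient and a view-angle fade.
# An inline world-space warped-noise and Fresnel color chain also exists but does not feed the emissive chain.
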Material /BRCosmetics/Weapons/FORT_Melee/Pickaxe_DualParadox/FX/M_Pickaxe_DualParadox_little_spikes
graph {
 properties {
  bUsedWithNiagaraMeshParticles=true
  blendMode=BLEND_Translucent
  pixelDepthOffsetMode=PDOM_Legacy
  previewMesh=&/SM_Pickaxe_DualParadox_little_spikes
  shadingModel=MSM_Unlit
  usedShadingModels=Unlit
 }
 nodes [
  n1 = ComponentMask(None=VectorParameter() {parameterName=DualParadox_PannerDirection, defaultValue=<rgba 0 -0.05 0 0>}.RGB) {g=true, r=true}
  n2 = Divide(A=ComponentMask(None=WorldPosition().XYZ) {b=true, g=true}, B=ScalarParameter() {defaultValue=50, parameterName=DualParadox_NoiseSize}) {constB=100}
  n3 = ComponentMask(None=TextureSample(UVs=@n2) {samplerType=SAMPLERTYPE_LinearColor, texture=/Game/Effects/Fort_Effects/Textures/Plasma/T_Turbulence_AdditiveWarp_Linear}.RGB) {g=true, r=true}
  n4 = ScalarParameter() {defaultValue=0.13, parameterName=DualParadox_Distortion}
  n5 = TextureSample(UVs=Add(A=Panner(Coordinate=@n2, Speed=@n1), B=Multiply(A=@n3, B=@n4))) {samplerType=SAMPLERTYPE_Masks, texture=/Game/Global/Textures/Default/Noise/NoisePattern}
  n6 = Transform(None=Constant3Vector() {constant=<rgba 0 0 1 0>})
  n7 = CameraVectorWS()
  n8 = MaterialFunctionCall(Z=ScalarParameter() {defaultValue=0.7, parameterName=DualParadox_ReduceZFresnel}) {materialFunction=&MakeFloat3}
  LinearInterpolate(
   A=LinearInterpolate(
    A=VectorParameter() {parameterName=DualParadox_Color1, defaultValue=<rgba 0.7454 0.2232 1 1>}.RGB
    B=VectorParameter() {parameterName=DualParadox_Color2, defaultValue=<rgba 0.5149 0.117 0.7991 1>}.RGB
    Alpha=Sine(None=@n5.R)
   )
   B=VectorParameter() {parameterName=DualParadox_FresnelColor, defaultValue=<rgba 2.5997 1.9531 3 1>}.RGB
   Alpha=Fresnel(
    ExponentIn=ScalarParameter() {defaultValue=3, parameterName=DualParadox_FresnelExponent}
    Normal=LinearInterpolate(A=@n6, B=@n7, Alpha=@n8.Result)
   ) {baseReflectFraction=0}
  )
  n9 = ParticleColor()
  MaterialFunctionCall(Emissive=Multiply(
    A=@n9.RGB
    B=MaterialFunctionCall!(MA, Position, Normal, Fresnel Normal) {materialFunction=/BRCosmetics/Weapons/FORT_Melee/Pickaxe_DualParadox/Materials/MFC_DualParadox_PurpleNoise}.Color
   )) {materialFunction=&MF_FX_EmissiveAdjustments}
  n10 = TextureCoordinate()
  n11 = Multiply(A=OneMinus(None=ComponentMask(None=@n10) {g=true})) {constB=20}
  Multiply(
   A=Multiply(A=@n9.A, B=Saturate(None=@n11))
   B=MaterialFunctionCall("FadeSensitivity (optional)"=ScalarParameter() {parameterName="FadeSensitivity (optional) (S)"}) {materialFunction=&FadeBasedOnViewAngle}.Result
  )
  ScalarParameter() {defaultValue=0.1, parameterName=DualParadox_PhasingSpeed}
 ]
}
