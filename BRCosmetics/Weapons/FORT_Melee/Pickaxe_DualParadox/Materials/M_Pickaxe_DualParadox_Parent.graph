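# Parent material for the Dual Paradox pickaxe, layering a purple noise effect onto the shared character shading functions.
# BaseFN PreFX builds the material attributes, which pass through MFC_DualParadox_PurpleNoise with the transformed normal as its Fresnel normal.
# The result runs through BaseFN PostFX, which also reads PreFX outputs, then CINE_MF_CharPostFX.
# It uses material attributes and supports skeletal mesh, Nanite, morph targets and clothing.
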Material /BRCosmetics/Weapons/FORT_Melee/Pickaxe_DualParadox/Materials/M_Pickaxe_DualParadox_Parent
graph {
 properties {
  bAutomaticallySetUsageInEditor=false
  bUseMaterialAttributes=true
  bUsedWithClothing=true
  bUsedWithMeshDeformer=true
  bUsedWithMorphTargets=true
  bUsedWithNanite=true
  bUsedWithSkeletalMesh=true
  pixelDepthOffsetMode=PDOM_Legacy
  shadingModel=MSM_Unlit
  subsurfaceProfile=&SS_HeroSkin_02
  usedShadingModels=Unlit
 }
 nodes [
  n1 = MaterialFunctionCall() {materialFunction=&BaseFN_CharacterFunction_PreFX}
  n2 = GetMaterialAttributes(None=@n1.Result) {attributeGetTypes [#1]}
  n3 = MaterialFunctionCall(
   AO=@n1.AO
   Distance Blend=@n1.AO
   Fresnel=@n1.AO
   Gmap=@n1.AO
   Pre Skinned Local Position=@n1.AO
   Material=MaterialFunctionCall!(MA=@n2.MaterialAttributes, Position, Normal, Fresnel Normal=Transform(None=@n2.Normal)) {materialFunction=&/MFC_DualParadox_PurpleNoise}.Result
  ) {materialFunction=&BaseFN_CharacterFunction_PostFX}
  MaterialFunctionCall(In=@n3.Result) {materialFunction=&CINE_MF_CharPostFX}
 ]
}
